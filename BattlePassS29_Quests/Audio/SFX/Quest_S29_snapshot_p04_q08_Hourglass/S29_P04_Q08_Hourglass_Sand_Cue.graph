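# Looping hourglass sand ambience SoundCue for a Battle Pass Season 29 quest, routed to the Fort_Quests sound class.
# A Mixer layers sand, static and tonal beds, with the static layer mixed lower.
# Each layer randomly picks one of three looping waves and passes it through a Modulator for slight pitch and volume variation.
# Overridden attenuation applies natural-sound falloff and HRTF spatialization, with a boosted overall volume.
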SoundCue /BattlePassS29_Quests/Audio/SFX/Quest_S29_snapshot_p04_q08_Hourglass/S29_P04_Q08_Hourglass_Sand_Cue
root=Mixer_0
nodes [
 {id=Mixer_0, children [Modulator_1,Modulator_2,Modulator_3], properties {inputVolume [1,0.6,1]}}
 {id=Modulator_1, children [Random_1], properties {pitchMax=1.02, pitchMin=0.98, volumeMax=1.02, volumeMin=0.98}}
 {id=Modulator_2, children [Random_2], properties {pitchMax=1.02, pitchMin=0.98, volumeMax=1.02, volumeMin=0.98}}
 {id=Modulator_3, children [Random_3], properties {pitchMax=1.02, pitchMin=0.98, volumeMax=1.02, volumeMin=0.98}}
 {id=Random_1, children [WavePlayer_6,WavePlayer_7,WavePlayer_8], properties {weights [1,1,1]}}
 {id=Random_2, children [WavePlayer_9, WavePlayer_10, WavePlayer_11], properties {weights [1,1,1]}}
 {id=Random_3, children [WavePlayer_12, WavePlayer_13, WavePlayer_14], properties {weights [1,1,1]}}
 {
  id=WavePlayer_10
  properties {bLooping=true, soundWaveAssetPtr=&/S29_P04_Q08_Hourglass_Static_LP_02}
 }
 {
  id=WavePlayer_11
  properties {bLooping=true, soundWaveAssetPtr=&/S29_P04_Q08_Hourglass_Static_LP_03}
 }
 {
  id=WavePlayer_12
  properties {bLooping=true, soundWaveAssetPtr=&/S29_P04_Q08_Hourglass_Tonal_LP_01}
 }
 {
  id=WavePlayer_13
  properties {bLooping=true, soundWaveAssetPtr=&/S29_P04_Q08_Hourglass_Tonal_LP_02}
 }
 {
  id=WavePlayer_14
  properties {bLooping=true, soundWaveAssetPtr=&/S29_P04_Q08_Hourglass_Tonal_LP_03}
 }
 {
  id=WavePlayer_6
  properties {bLooping=true, soundWaveAssetPtr=&/S29_P04_Q08_Hourglass_Sand_LP_01}
 }
 {
  id=WavePlayer_7
  properties {bLooping=true, soundWaveAssetPtr=&/S29_P04_Q08_Hourglass_Sand_LP_02}
 }
 {
  id=WavePlayer_8
  properties {bLooping=true, soundWaveAssetPtr=&/S29_P04_Q08_Hourglass_Sand_LP_03}
 }
 {
  id=WavePlayer_9
  properties {bLooping=true, soundWaveAssetPtr=&/S29_P04_Q08_Hourglass_Static_LP_01}
 }
]
properties {
 attenuationSettings=/BattlePassS29_Quests/Audio/SFX/Quest_S29_snapshot_p03_Digsite/S29_Quests_Atten
 bOverrideAttenuation=true
 maxDistance=3314.2856
 soundClassObject=&Fort_Quests
 volumeMultiplier=2
 attenuationOverrides {
  dBAttenuationAtMax=-28
  distanceAlgorithm=NaturalSound
  falloffDistance=1500
  spatializationAlgorithm=SPATIALIZATION_HRTF
  attenuationShapeExtents {x=100}
 }
}
